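# Sound cue for the basic armored truck's engine shutoff.
# A mixer plays two wave layers together at equal volume, a MERC C63 shutoff and an armored truck Peterbuilt off recording.
# The cue boosts overall volume and uses the shared vehicle movement attenuation, the vehicle engine concurrency set and the vehicle interiors sound class.
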SoundCue /ArmoredTruck/Assets/Audio/Veh_BasicArmoredTruck_Engine_Off_Cue
root=Mixer_0
nodes [
 {id=Mixer_0, children [WavePlayer_0,WavePlayer_2], properties {inputVolume [1,1]}}
 {id=WavePlayer_0, properties {soundWaveAssetPtr=&/MERC_C63_Shutoff_Mono}}
 {id=WavePlayer_2, properties {soundWaveAssetPtr=&/S34_ArmoredTruck_Peterbuilt_Off}}
]
properties {
 attenuationSettings=/VehicleAudio/Settings/Vehicle_Movement_Att
 maxDistance=12512
 soundClassObject=&Fort_Vehicles_Interior
 volumeMultiplier=3
 concurrencySet [/VehicleAudio/Settings/Vehicle_Engine_Conc]
}
