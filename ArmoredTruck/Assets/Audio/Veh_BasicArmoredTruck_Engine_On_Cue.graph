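# Sound cue for the basic armored truck's engine-on sound, built from a single Peterbuilt-named wave.
# The wave player feeds the output directly, with no randomization, mixing or looping nodes in the cue.
# It plays in the Fort vehicle interior sound class with boosted volume, shared vehicle movement attenuation and a vehicle engine concurrency limit.
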SoundCue /ArmoredTruck/Assets/Audio/Veh_BasicArmoredTruck_Engine_On_Cue
root=WavePlayer_1
nodes [{id=WavePlayer_1, properties {soundWaveAssetPtr=&/S34_ArmoredTruck_Peterbuilt_On}}]
properties {
 attenuationSettings=/VehicleAudio/Settings/Vehicle_Movement_Att
 maxDistance=12512
 soundClassObject=&Fort_Vehicles_Interior
 volumeMultiplier=2.5
 concurrencySet [/VehicleAudio/Settings/Vehicle_Engine_Conc]
}
