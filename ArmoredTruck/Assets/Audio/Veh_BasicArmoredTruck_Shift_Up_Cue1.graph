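# Sound cue for the armored truck's upshift, using close-detail vehicle attenuation and the vehicle interior sound class.
# A Mixer layers three randomized parts: mechanical gear shifts, slightly quieter exhaust pops after a short random delay, and truck-specific upshift sounds.
# Modulators vary the first two layers, while the upshift layer keeps a fixed pitch.
# A quality switch gives that layer either five variations or a three-wave subset to choose from.
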SoundCue /ArmoredTruck/Assets/Audio/Veh_BasicArmoredTruck_Shift_Up_Cue1
root=Mixer_0
nodes [
 {id=Delay_0, children [Modulator_1], properties {delayMax=0.1, delayMin=0.05}}
 {id=Mixer_0, children [Modulator_0,Delay_0,Modulator_3], properties {inputVolume [1,0.8,1]}}
 {id=Modulator_0, children [Random_1]}
 {id=Modulator_1, children [Random_0]}
 {id=Modulator_3, children [QualityLevel_0], properties {pitchMax=1, pitchMin=1}}
 {id=QualityLevel_0, children [Random_3,Random_3,Random_4,Random_4]}
 {id=Random_0, children [WavePlayer_7, WavePlayer_8, WavePlayer_9, WavePlayer_10], properties {weights [1,1,1,1]}}
 {id=Random_1, children [WavePlayer_2,WavePlayer_3,WavePlayer_4], properties {weights [1,1,1]}}
 {id=Random_3, children [WavePlayer_12, WavePlayer_13, WavePlayer_14, WavePlayer_15, WavePlayer_16], properties {weights [1,1,1,1,1]}}
 {id=Random_4, children [WavePlayer_14, WavePlayer_15, WavePlayer_16], properties {bShouldExcludeFromBranchCulling=true, weights [1,1,1]}}
 {id=WavePlayer_10, properties {soundWaveAssetPtr=&/MERC_C63_ExhaustPop_04}}
 {id=WavePlayer_12, properties {soundWaveAssetPtr=&/S34_ArmoredTruck_Shifts_Up_01}}
 {id=WavePlayer_13, properties {soundWaveAssetPtr=&/S34_ArmoredTruck_Shifts_Up_02}}
 {id=WavePlayer_14, properties {soundWaveAssetPtr=&/S34_ArmoredTruck_Shifts_Up_03}}
 {id=WavePlayer_15, properties {soundWaveAssetPtr=&/S34_ArmoredTruck_Shifts_Up_04}}
 {id=WavePlayer_16, properties {soundWaveAssetPtr=&/S34_ArmoredTruck_Shifts_Up_05}}
 {id=WavePlayer_2, properties {soundWaveAssetPtr=/VehicleAudio/Assets/Gears/Veh_Valet_Gen_Gear_Shift_Mech_03}}
 {id=WavePlayer_3, properties {soundWaveAssetPtr=/VehicleAudio/Assets/Gears/Veh_Valet_Gen_Gear_Shift_Mech_01}}
 {id=WavePlayer_4, properties {soundWaveAssetPtr=/VehicleAudio/Assets/Gears/Veh_Valet_Gen_Gear_Shift_Mech_02}}
 {id=WavePlayer_7, properties {soundWaveAssetPtr=&/MERC_C63_ExhaustPop_01}}
 {id=WavePlayer_8, properties {soundWaveAssetPtr=&/MERC_C63_ExhaustPop_02}}
 {id=WavePlayer_9, properties {soundWaveAssetPtr=&/MERC_C63_ExhaustPop_03}}
]
properties {
 attenuationSettings=/VehicleAudio/Settings/Vehicle_Movement_Close_Detail_Att
 maxDistance=12000
 soundClassObject=&Fort_Vehicles_Interior
 volumeMultiplier=1
}
