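# Material function that adds a scorched, damaged look to incoming material attributes, apparently for the armored truck.
# An FX mask texture supplies scorch (R) and normal (G) masks for darkening base color, reducing metallic, shifting roughness and blending in a tiled damage normal with BlendAngleCorrectedNormals.
# The static switch UseScortch enables the effect, with ScorchAmount blending between material attribute sets.
# Named reroutes connect the stages.
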MaterialFunction /ArmoredTruck/Assets/Materials/MF_ArmoredTruck_Scortch
graph {nodes [
  n1 = NamedRerouteUsage()
  n2 = NamedRerouteUsage()
  n3 = ScalarParameter() {group=Scorch, parameterName=ScorchAmount}
  FunctionOutput(None=StaticSwitchParameter(True=BlendMaterialAttributes(A=@n1, B=@n2, Alpha=@n3), False=NamedRerouteUsage()) {group=Scorch, parameterName=UseScortch})
  NamedRerouteDeclaration(None=FunctionInput() {inputType=FunctionInput_MaterialAttributes}) {name=InMaterialAttributes, nodeColor=<rgba 0 0.5547 1 1>}
  n4 = GetMaterialAttributes(None=NamedRerouteUsage()) {attributeGetTypes [
    #1
    #2
    #3
    #4
   ]}
  NamedRerouteDeclaration(None=@n4.Normal) {name=InNormal, nodeColor=<rgba 0.875 1 0 1>}
  NamedRerouteDeclaration(None=@n4.Roughness) {name=InRoughness, nodeColor=<rgba 0 1 0.3203 1>}
  NamedRerouteDeclaration(None=LinearInterpolate(A=NamedRerouteUsage(), Alpha=NamedRerouteUsage()) {constB=0}) {name=OutMetallic, nodeColor=<rgba 0.6172 0 1 1>}
  NamedRerouteDeclaration(None=LinearInterpolate(A=NamedRerouteUsage(), B=ScalarParameter() {defaultValue=0.9, group=Scorch, parameterName=ScortchRoughness}, Alpha=NamedRerouteUsage()) {constB=0.9}) {name=OutRoughness, nodeColor=<rgba 0.6406 1 0 1>}
  n5 = SetMaterialAttributes!(
   MaterialAttributes=NamedRerouteUsage()
   BaseColor=NamedRerouteUsage()
   Metallic=NamedRerouteUsage()
   Roughness=NamedRerouteUsage()
   Normal=NamedRerouteUsage()
  ) {attributeSetTypes [
    #1
    #2
    #3
    #4
   ]}
  NamedRerouteDeclaration(None=@n5) {name=OutScortchMaterial, nodeColor=<rgba 1 0 0.1641 1>}
  n6 = NamedRerouteUsage()
  n7 = TextureSampleParameter2D(UVs=Multiply(A=TextureCoordinate(), B=ScalarParameter() {defaultValue=2, group=Scorch, parameterName=BrokenNormalTiling})) {
   group=Scorch
   parameterName=ScorchDamageNormal
   samplerType=SAMPLERTYPE_Normal
   texture=/Game/Global/Textures/Default/Tileable_Patterns/T_Tilling_Distortion_Map_01_n
  }
  NamedRerouteDeclaration(None=LinearInterpolate(
    A=@n6
    B=MaterialFunctionCall(BaseNormal=@n6, AdditionalNormal=@n7.RGB) {materialFunction=&BlendAngleCorrectedNormals}.Result
    Alpha=NamedRerouteUsage()
   )) {name=OutNormals, nodeColor=<rgba 1 0.2812 0 1>}
  NamedRerouteDeclaration(None=@n4.BaseColor) {name=InBaseColor, nodeColor=<rgba 0.8047 0 1 1>}
  n8 = TextureSampleParameter2D() {group=Scorch, parameterName=ScorchMask, samplerType=SAMPLERTYPE_Masks, texture=/ArmoredTruck/Assets/Textures/T_ArmoredTruck_FX}
  NamedRerouteDeclaration(None=@n8.R) {name=ScortchMask, nodeColor=<rgba 0 1 0.4844 1>}
  NamedRerouteDeclaration(None=@n8.G) {name=NormalMapMask, nodeColor=<rgba 1 0.1641 0 1>}
  n9 = NamedRerouteUsage()
  n10 = OneMinus(None=ScalarParameter() {defaultValue=0.75, group=Scorch, parameterName=ScortchIntensity})
  NamedRerouteDeclaration(None=LinearInterpolate(A=NamedRerouteUsage(), B=Multiply(A=@n9, B=@n10) {constB=0.1}, Alpha=NamedRerouteUsage())) {name=OutBaseColor, nodeColor=<rgba 1 0.9609 0 1>}
  NamedRerouteDeclaration(None=@n4.Metallic) {name=InMetallic, nodeColor=<rgba 0.6406 0 1 1>}
 ]}
