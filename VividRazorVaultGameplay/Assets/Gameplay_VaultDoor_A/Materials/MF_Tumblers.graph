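# Material function for the vault door's tumbler weakpoints, layering overheat, damage and hit effects onto incoming material attributes when UseOverheat is on.
# Vertex color values identify five tumblers for per-tumbler hide, damage and highlight parameters, with hiding done through world position offset.
# Noise-broken gradient, normal and radial sphere masks, animated by timestamp parameters, shape the heat glow across tumblers and cover.
# These masks feed emissive, base color and roughness, with separate emissive scaling for ray tracing.
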MaterialFunction /VividRazorVaultGameplay/Assets/Gameplay_VaultDoor_A/Materials/MF_Tumblers
graph {nodes [
  n1 = Multiply(A=Time()) {constB=0.3}
  ConstantBiasScale(None=Sine(None=@n1))
  Constant() {r=1}
  n2 = NamedRerouteUsage()
  Divide(A=Multiply(A=@n2, B=MaterialFunctionCall() {materialFunction=&Pi}.Result)) {desc=EaseOutCurve}
  EyeAdaptation()
  Frac(None=@n1)
  n3 = GetMaterialAttributes(None=NamedRerouteDeclaration(None=FunctionInput() {inputType=FunctionInput_MaterialAttributes}) {name=InMaterialAttributes, nodeColor=<rgba 0.9922 1 0 1>}) {attributeGetTypes [#1, #2, #3]}
  n4 = SetMaterialAttributes!(
   MaterialAttributes=@n3.MaterialAttributes
   WorldPositionOffset=NamedRerouteUsage()
   EmissiveColor=NamedRerouteUsage()
   BaseColor=NamedRerouteUsage()
   Roughness=NamedRerouteUsage()
  ) {attributeSetTypes [
    #4
    #1
    #2
    #3
   ]}
  FunctionOutput(None=StaticSwitchParameter(True=@n4, False=NamedRerouteUsage()) {group=TumblerWeakpoint, parameterName=UseOverheat})
  n5 = Multiply(A=OneMinus(None=Fresnel() {baseReflectFraction=0, exponent=0.5}), B=NamedRerouteUsage())
  n6 = NamedRerouteUsage()
  n7 = LinearInterpolate(B=ScalarParameter() {defaultValue=0.2, group=TumblerWeakpoint, parameterName=GradientSmoothness}, Alpha=@n6) {constA=-0.8, constB=0.2}
  n8 = ScalarParameter() {defaultValue=1.2, group=TumblerWeakpoint, parameterName=GradientSmoothness2}
  n9 = TextureSampleParameter2D() {
   group=TumblerWeakpoint
   parameterName=TumblerGradientMask
   samplerType=SAMPLERTYPE_LinearGrayscale
   texture=/VividRazorVaultGameplay/Assets/Gameplay_VaultDoor_A/Textures/T_Gameplay_VaultDoor_A_A_M
  }
  n10 = ScalarParameter() {defaultValue=2, group=TumblerWeakpoint, parameterName=GradientIntensity}
  n11 = Multiply(
   A=Subtract(
    A=NamedRerouteDeclaration(None=@n9.G) {name=TumblerGradient, nodeColor=<rgba 0 1 0.6484 1>}
    B=ScalarParameter() {defaultValue=0.5, group=TumblerWeakpoint, parameterName=GradientOffset}
   ) {constB=0.5}
   B=@n10
  ) {constB=2}
  n12 = Divide(
   A=NamedRerouteDeclaration(None=Abs(None=@n11)) {name=TipsGradient, nodeColor=<rgba 1 0.3516 0 1>}
   B=LinearInterpolate(B=@n10, Alpha=@n6) {constA=1}
  )
  n13 = Saturate(None=OneMinus(None=SmoothStep(Min=@n7, Max=@n8, Value=@n12) {constMin=0.25}))
  n14 = Multiply(A=TextureCoordinate(), B=ScalarParameter() {defaultValue=1, group=TumblerWeakpoint, parameterName=NoiseTiling})
  n15 = TextureSample(UVs=Multiply(A=@n14) {constB=10}) {texture=/Game/Environments/Helios/Terrain/Textures/T_Helios_Landscape_Noise}
  n16 = Multiply(A=@n15.G) {constB=0.5}
  n17 = Multiply(A=@n5, B=Saturate(None=Subtract(A=@n13, B=@n16)))
  LinearInterpolate(
   A=LinearInterpolate(B=ScalarParameter() {defaultValue=0.17, group=TumblerWeakpoint, parameterName=RedFresnel}, Alpha=@n5) {constB=0.01}
   B=@n17
   Alpha=SmoothStep(Value=@n6) {constMin=0.25}
  )
  MaterialFunctionCall() {materialFunction=/Game/Effects/Niagara/MasterMaterials/Functions/InstanceBounds_0-1_UVW}
  n18 = Subtract(A=OneMinus(None=NamedRerouteUsage()))
  n19 = NamedRerouteUsage()
  n20 = ChannelMaskParameter(None=Transform(None=VertexNormalWS()) {transformSourceType=TRANSFORMSOURCE_World, transformType=TRANSFORM_Local}) {group=TumblerWeakpoint, maskChannel=Green, parameterName=TumblerGlowNormalMask, defaultValue=<rgba 0 1 0 0>}
  n21 = ConstantBiasScale(None=@n20.R)
  n22 = Power(
   Base=LinearInterpolate(
    A=@n21
    B=OneMinus(None=@n21)
    Alpha=ScalarParameter() {defaultValue=1, group=TumblerWeakpoint, parameterName=CoverInvertNormal, sliderMax=1}
   )
   Exp=ScalarParameter() {defaultValue=1, group=TumblerWeakpoint, parameterName=CoverNormalPowerAdjust}
  )
  n23 = ScalarParameter() {defaultValue=5.0501, group=TumblerWeakpoint, parameterName=CoverNormalIntensityAdjust}
  n24 = Subtract(A=NamedRerouteUsage(), B=ScalarParameter() {defaultValue=0.2, group=TumblerWeakpoint, parameterName=CoverTipsMaskAdjust})
  n25 = ScalarParameter() {defaultValue=6.658, group=TumblerWeakpoint, parameterName=CoverTipsMaskPower}
  n26 = Add(A=Saturate(None=Multiply(A=@n22, B=@n23)), B=Saturate(None=Power(Base=@n24, Exp=@n25) {constExponent=8}))
  n27 = Multiply(A=ScalarParameter() {defaultValue=0.4485, group=TumblerWeakpoint, parameterName=CoverOverheatGradientIntensity}, B=NamedRerouteUsage())
  n28 = Saturate(None=Saturate(None=Multiply(A=@n26, B=@n27)))
  n29 = Subtract(A=@n28, B=Multiply(A=NamedRerouteUsage(), B=NamedRerouteUsage()))
  n30 = NamedRerouteUsage()
  n31 = Add(A=Saturate(None=Add(A=@n28, B=@n29)), B=LinearInterpolate(Alpha=SmoothStep(Value=@n30) {constMin=0.25}) {constB=0.05})
  Multiply(A=Saturate(None=Add(A=@n18, B=@n19)), B=@n31)
  n32 = Multiply(
   A=Multiply(A=LocalPosition().XYZ) {constB=0.001}
   B=ScalarParameter() {defaultValue=0.5, group=TumblerWeakpoint, parameterName=RadialMaskNoiseSize}
  ) {constB=0.5}
  n33 = ComponentMask(None=@n32) {b=true, r=true}
  n34 = ComponentMask(None=@n32) {b=true, g=true}
  n35 = ScalarParameter() {group=TumblerWeakpoint, parameterName=RadialMaskGBAxis, sliderMax=1}
  n36 = TextureSample(UVs=LinearInterpolate(
    A=LinearInterpolate(A=@n33, B=@n34, Alpha=@n35)
    B=ComponentMask(None=@n32) {g=true, r=true}
    Alpha=ScalarParameter() {group=TumblerWeakpoint, parameterName=RadialMaskRGAxis, sliderMax=1}
   )) {texture=&T_Helios_Landscape_Noise}
  Multiply(A=@n36.B)
  n37 = ScalarParameter() {group=TumblerWeakpoint, parameterName=HideTumbler1}
  n38 = NamedRerouteUsage()
  n39 = Constant() {r=0.2}
  n40 = ScalarParameter() {group=TumblerWeakpoint, parameterName=HideTumbler2}
  n41 = NamedRerouteUsage()
  n42 = Constant() {r=0.4}
  n43 = Add(
   A=Add(
    A=MaterialFunctionCall!(Hide Element Toggle=@n37, VertexColorChannel=@n38, Vertex Value to hide=@n39) {materialFunction=/Game/Packages/Fortress_SharedMaterials/Base_Material_Functions/VertexColorMask}.Result
    B=MaterialFunctionCall(Hide Element Toggle=@n40, VertexColorChannel=@n41, Vertex Value to hide=@n42) {materialFunction=&VertexColorMask}.Result
   )
   B=MaterialFunctionCall(
    Hide Element Toggle=ScalarParameter() {group=TumblerWeakpoint, parameterName=HideTumbler3}
    VertexColorChannel=NamedRerouteUsage()
    Vertex Value to hide=Constant() {r=0.6}
   ) {materialFunction=&VertexColorMask}.Result
  )
  n44 = MaterialFunctionCall(
   Hide Element Toggle=ScalarParameter() {group=TumblerWeakpoint, parameterName=HideTumbler4}
   VertexColorChannel=NamedRerouteUsage()
   Vertex Value to hide=Constant() {r=0.8}
  ) {materialFunction=&VertexColorMask}
  n45 = ScalarParameter() {group=TumblerWeakpoint, parameterName=HideTumbler5}
  n46 = NamedRerouteUsage()
  n47 = Constant() {r=1}
  n48 = Saturate(None=Add(
    A=Add(A=@n43, B=@n44.Result)
    B=MaterialFunctionCall(Hide Element Toggle=@n45, VertexColorChannel=@n46, Vertex Value to hide=@n47) {materialFunction=&VertexColorMask}.Result
   ))
  NamedRerouteDeclaration(None=@n48) {name=WPOHideMask, nodeColor=<rgba 1 0.2344 0 1>}
  n49 = NamedRerouteDeclaration(None=Multiply(A=ScalarParameter() {group=TumblerWeakpoint, parameterName=DamageTumbler1, sliderMax=1}, B=NamedRerouteUsage())) {name=DamageTumbler1, nodeColor=<rgba 1 0 0.2344 1>}
  n50 = NamedRerouteDeclaration(None=Multiply(A=ScalarParameter() {group=TumblerWeakpoint, parameterName=DamageTumbler2, sliderMax=1}, B=NamedRerouteUsage())) {name=DamageTumbler2, nodeColor=<rgba 0 0.5547 1 1>}
  n51 = Multiply(A=ScalarParameter() {group=TumblerWeakpoint, parameterName=DamageTumbler3, sliderMax=1}, B=NamedRerouteUsage())
  n52 = ScalarParameter() {group=TumblerWeakpoint, parameterName=DamageTumbler4, sliderMax=1}
  n53 = NamedRerouteUsage()
  n54 = Add(
   A=Add(A=Add(A=@n49, B=@n50), B=NamedRerouteDeclaration(None=@n51) {name=DamageTumbler3, nodeColor=<rgba 0.875 1 0 1>})
   B=NamedRerouteDeclaration(None=Multiply(A=@n52, B=@n53)) {name=DamageTumbler4, nodeColor=<rgba 0.8047 0 1 1>}
  )
  n55 = NamedRerouteDeclaration(None=Multiply(A=ScalarParameter() {group=TumblerWeakpoint, parameterName=DamageTumbler5, sliderMax=1}, B=NamedRerouteUsage())) {name=DamageTumbler5, nodeColor=<rgba 0 1 0.4844 1>}
  NamedRerouteDeclaration(None=Saturate(None=Add(A=@n54, B=@n55))) {name=TumblerDamageMask, nodeColor=<rgba 1 0.5625 0 1>}
  n56 = EyeAdaptationInverse()
  n57 = VectorParameter() {group=TumblerWeakpoint, parameterName=Param_2, defaultValue=<rgba 2.5 0.0474 0 1>}
  n58 = VectorParameter() {group=TumblerWeakpoint, parameterName=Param_1, defaultValue=<rgba 100 11.2364 0 1>}
  n59 = LinearInterpolate(A=NamedRerouteUsage(), Alpha=NamedRerouteUsage()) {constA=1}
  n60 = NamedRerouteUsage()
  n61 = Multiply(A=Saturate(None=Add(A=@n2, B=@n60)), B=ScalarParameter() {defaultValue=1, group=TumblerWeakpoint, parameterName=ThermiteOverheatIntensity}) {constB=0.75}
  n62 = Add(A=LinearInterpolate(A=NamedRerouteUsage(), B=LinearInterpolate(A=@n57.RGB, B=@n58.RGB, Alpha=@n59), Alpha=@n61), B=NamedRerouteUsage())
  n63 = ScalarParameter() {defaultValue=25, group=TumblerWeakpoint, parameterName=LumenContributionIntensity}
  NamedRerouteDeclaration(None=RayTracingQualitySwitch(Normal=Multiply(A=@n56.EyeAdaptationInverse, B=@n62), RayTraced=Multiply(A=@n62, B=@n63) {constB=25})) {name=OutEmissive, nodeColor=<rgba 0 0.8828 1 1>}
  NamedRerouteDeclaration(None=@n3.EmissiveColor) {name=InEmissive, nodeColor=<rgba 1 0.7969 0 1>}
  NamedRerouteDeclaration(None=@n3.BaseColor) {name=InBaseColor, nodeColor=<rgba 0.4766 0 1 1>}
  n64 = Saturate(None=@n61)
  NamedRerouteDeclaration(None=LinearInterpolate(A=NamedRerouteUsage(), Alpha=@n64) {constB=0}) {name=OutBaseColor, nodeColor=<rgba 0 1 0.1562 1>}
  n65 = Subtract(A=WorldPosition().XYZ, B=ObjectPositionWS())
  NamedRerouteDeclaration(None=Multiply(A=Multiply(A=@n65) {constB=-1}, B=NamedRerouteUsage())) {name=OutWPO, nodeColor=<rgba 1 0.9609 0 1>}
  n66 = Time()
  n67 = ScalarParameter() {group=TumblerWeakpoint, parameterName=HitTimestamp}
  n68 = Constant() {r=0.4}
  n69 = Divide(A=Subtract(A=@n66, B=@n67), B=@n68) {constB=0.15}
  n70 = Multiply(A=OneMinus(None=Clamp(None=@n69)), B=NamedRerouteUsage())
  n71 = Subtract(A=@n66, B=Add(A=@n67, B=Add(A=@n68) {constB=0.5, desc=eXTRAdURATION}))
  n72 = OneMinus(None=Clamp(None=Divide(A=@n71) {constB=0.5, desc=OutLength}))
  NamedRerouteDeclaration(None=VertexInterpolator(VS=Multiply(A=@n70, B=@n72)).PS) {name=HitTimestampMask, nodeColor=<rgba 0 1 1 1>}
  n73 = OneMinus(None=NamedRerouteUsage())
  n74 = Fresnel() {baseReflectFraction=0, exponent=0.1}
  n75 = Multiply(A=Multiply(A=@n73, B=@n73), B=Multiply(A=NamedRerouteUsage(), B=OneMinus(None=@n74)))
  n76 = VectorParameter() {group=TumblerWeakpoint, parameterName=HitColor, defaultValue=<rgba 75 37.4982 7.4146 1>}
  NamedRerouteDeclaration(None=StaticSwitchParameter(True=Multiply(A=@n75, B=@n76.RGB), False=Constant()) {group=TumblerWeakpoint, parameterName=UseTumblers}) {name=HitHighlight, nodeColor=<rgba 0.7812 1 0 1>}
  n77 = NamedRerouteUsage()
  n78 = Saturate(None=Subtract(A=NamedRerouteUsage(), B=@n77))
  n79 = ScalarParameter() {group=TumblerWeakpoint, parameterName=UseCoverOverheatTimestamp}
  NamedRerouteDeclaration(None=StaticSwitchParameter(True=LinearInterpolate(A=@n77, B=@n78, Alpha=@n79), False=Constant()) {group=TumblerWeakpoint, parameterName=UseTumblers}) {name=TumblerOverheatMask, nodeColor=<rgba 0 0.6484 1 1>}
  n80 = ScalarParameter() {group=TumblerWeakpoint, parameterName=OverheatTimestamp}
  n81 = Divide(A=Subtract(A=Time(), B=@n80), B=ScalarParameter() {defaultValue=6, group=TumblerWeakpoint, parameterName=OverheatTimestampDuration})
  n82 = NamedRerouteDeclaration(None=Multiply(A=Clamp(None=@n81), B=Saturate(None=@n80))) {name=OverheatTimestamp, nodeColor=<rgba 0 0.8125 1 1>}
  n83 = NamedRerouteUsage()
  NamedRerouteDeclaration(None=VertexInterpolator(VS=Multiply(A=@n82, B=@n83)).PS) {name=CoverOverheatTimestamp, nodeColor=<rgba 0.0078 1 0 1>}
  n84 = LocalPosition()
  n85 = VectorParameter() {group=TumblerWeakpoint, parameterName=OverallHeatMaskPosition, defaultValue=<rgba 35 30 170 1>}
  n86 = ScalarParameter() {group=TumblerWeakpoint, parameterName=OverallOverheatAmount, sliderMax=1}
  n87 = NamedRerouteUsage()
  n88 = LinearInterpolate(A=@n87, B=OneMinus(None=@n87), Alpha=ScalarParameter() {group=TumblerWeakpoint, parameterName=OverheatTimestampInvert, sliderMax=1})
  n89 = ScalarParameter() {group=TumblerWeakpoint, parameterName=UseRadialOverheatTimestamp}
  n90 = LinearInterpolate(
   B=ScalarParameter() {defaultValue=154.8378, group=TumblerWeakpoint, parameterName=OverallOverheatRadius}
   Alpha=NamedRerouteDeclaration(None=LinearInterpolate(A=@n86, B=@n88, Alpha=@n89)) {name=OverallOverheatProgress, nodeColor=<rgba 1 0 0.4219 1>}
  ) {constB=400}
  n91 = Power(
   Base=SmoothStep(Min=@n36.R, Value=SphereMask(A=@n84.XYZ, B=@n85.RGB, Radius=@n90) {attenuationRadius=300, hardnessPercent=0})
   Exp=ScalarParameter() {defaultValue=1, group=TumblerWeakpoint, parameterName=RadialMaskPower}
  )
  NamedRerouteDeclaration(None=Multiply(A=@n91, B=NamedRerouteUsage())) {name=OverallOverheatMask, nodeColor=<rgba 1 0 0.6094 1>}
  n92 = OneMinus(None=Fresnel() {baseReflectFraction=0, exponent=1})
  n93 = NamedRerouteUsage()
  n94 = Power(
   Base=LinearInterpolate(A=@n93, B=OneMinus(None=@n93), Alpha=ScalarParameter() {group=TumblerWeakpoint, parameterName=RoughnessInvert})
   Exp=ScalarParameter() {defaultValue=1.5, group=TumblerWeakpoint, parameterName=RoughnessMaskPow}
  )
  n95 = ScalarParameter() {defaultValue=1.5, group=TumblerWeakpoint, parameterName=RoughnessMaskInt}
  n96 = NamedRerouteDeclaration(None=Saturate(None=Multiply(A=@n94, B=@n95))) {name=RoughnessMask, nodeColor=<rgba 0 0.3906 1 1>}
  n97 = Multiply(A=LinearInterpolate(A=Multiply(A=@n92, B=@n96), B=@n96), B=@n91)
  n98 = Multiply(A=Multiply(A=NamedRerouteUsage(), B=NamedRerouteUsage()), B=@n97) {constB=0.25}
  n99 = NamedRerouteUsage()
  n100 = Subtract(
   A=Saturate(None=SmoothStep(Min=@n98, Value=@n97))
   B=Multiply(A=Multiply(A=@n99, B=@n98), B=ScalarParameter() {defaultValue=3, group=TumblerWeakpoint, parameterName=ThermiteNoiseIntensity}) {constB=3}
  )
  n101 = ScalarParameter() {defaultValue=2, group=TumblerWeakpoint, parameterName=RadialMaskInteriorPower}
  NamedRerouteDeclaration(None=Saturate(None=Power(Base=@n100, Exp=@n101))) {name=OverallHeatGradient, nodeColor=<rgba 0.9688 0 1 1>}
  n102 = TextureSample(UVs=@n14) {texture=&T_Helios_Landscape_Noise}
  NamedRerouteDeclaration(None=LinearInterpolate(A=@n102.G, B=@n15.G, Alpha=@n102.G)) {name=HighFrequencyNoise, nodeColor=<rgba 0.7109 0 1 1>}
  NamedRerouteDeclaration(None=@n102.G) {name=LowFrequencyNoise, nodeColor=<rgba 1 0 0.1875 1>}
  n103 = NamedRerouteDeclaration(None=MaterialFunctionCall(
    Hide Element Toggle=ScalarParameter() {group=TumblerWeakpoint, parameterName=HighlightTumbler1}
    VertexColorChannel=NamedRerouteUsage()
    Vertex Value to hide=Constant() {r=0.2}
   ) {materialFunction=&VertexColorMask}.Result) {name=HighlightTumbler1, nodeColor=<rgba 1 0.1641 0 1>}
  n104 = NamedRerouteDeclaration(None=MaterialFunctionCall(
    Hide Element Toggle=ScalarParameter() {group=TumblerWeakpoint, parameterName=HighlightTumbler2}
    VertexColorChannel=NamedRerouteUsage()
    Vertex Value to hide=Constant() {r=0.4}
   ) {materialFunction=&VertexColorMask}.Result) {name=HighlightTumbler2, nodeColor=<rgba 0 0.1562 1 1>}
  n105 = MaterialFunctionCall(
   Hide Element Toggle=ScalarParameter() {group=TumblerWeakpoint, parameterName=HighlightTumbler3}
   VertexColorChannel=NamedRerouteUsage()
   Vertex Value to hide=Constant() {r=0.6}
  ) {materialFunction=&VertexColorMask}
  n106 = ScalarParameter() {group=TumblerWeakpoint, parameterName=HighlightTumbler4}
  n107 = NamedRerouteUsage()
  n108 = Constant() {r=0.8}
  n109 = Add(
   A=Add(A=Add(A=@n103, B=@n104), B=NamedRerouteDeclaration(None=@n105.Result) {name=HighlightTumbler3, nodeColor=<rgba 0.4766 1 0 1>})
   B=NamedRerouteDeclaration(None=MaterialFunctionCall(Hide Element Toggle=@n106, VertexColorChannel=@n107, Vertex Value to hide=@n108) {materialFunction=&VertexColorMask}.Result) {name=HighlightTumbler4, nodeColor=<rgba 1 0 0.7969 1>}
  )
  n110 = NamedRerouteDeclaration(None=MaterialFunctionCall(
    Hide Element Toggle=ScalarParameter() {group=TumblerWeakpoint, parameterName=HighlightTumbler5}
    VertexColorChannel=NamedRerouteUsage()
    Vertex Value to hide=Constant() {r=1}
   ) {materialFunction=&VertexColorMask}.Result) {name=HighlightTumbler5, nodeColor=<rgba 0 1 0.8828 1>}
  NamedRerouteDeclaration(None=Saturate(None=Add(A=@n109, B=@n110))) {name=TumblerHighlightMask, nodeColor=<rgba 0.4062 1 0 1>}
  NamedRerouteDeclaration(None=@n3.Roughness) {name=InRoughness, nodeColor=<rgba 0 0.6484 1 1>}
  NamedRerouteDeclaration(None=StaticSwitchParameter(True=@n17, False=Constant()) {group=TumblerWeakpoint, parameterName=UseTumblers}) {name=TumblersGrayscale, nodeColor=<rgba 0 0.8125 1 1>}
  NamedRerouteDeclaration(None=LinearInterpolate(A=NamedRerouteUsage(), Alpha=@n64) {constB=0.9}) {name=OutRoughness, nodeColor=<rgba 0.8047 0 1 1>}
  n111 = ChannelMaskParameter(None=VertexColor()) {group=TumblerWeakpoint, parameterName=VertexColorMask}
  NamedRerouteDeclaration(None=Ceil(None=NamedRerouteDeclaration(None=@n111.R) {name=GradientMasks, nodeColor=<rgba 1 0.6797 0 1>})) {name=TumblersMask, nodeColor=<rgba 0.7344 0 1 1>}
  NamedRerouteDeclaration(None=Multiply(A=@n31, B=NamedRerouteUsage())) {name=CoverOverheatMask, nodeColor=<rgba 0 1 0.9766 1>}
  NamedRerouteDeclaration(None=@n15.R) {name=MediumFrequencyNoise, nodeColor=<rgba 0.0459 0.0287 1 1>}
  NamedRerouteUsage()
  OneMinus(None=NamedRerouteUsage())
  ScalarParameter() {defaultValue=1, group=TumblerWeakpoint, parameterName=Testr, sliderMax=1}
  TextureSample() {samplerType=SAMPLERTYPE_Masks, texture=/VividRazorVaultGameplay/Assets/Gameplay_VaultDoor_A/Textures/T_Gameplay_VaultDoor_A_A_S}
  VertexInterpolator(VS=@n48)
 ]}
