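# Parent vehicle material for the Valet Armored Truck, using material attributes with Default Lit and Clear Coat shading.
# Valet base functionality feeds headlights and turn signals, split by pre-skinned bounds, then decals, reverse lights, damage, speakers, party lights and occlusion prevention.
# Optional material layers follow, then de-rez, vault tumbler and scorch functions.
# A UseSirenAlarm switch adds sine-pulsed, eye-adaptation-corrected emissive, using height-masked top lights, bump-offset bulbs, depth fade and per-primitive RedLights data.
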Material /ArmoredTruck/Assets/Materials/M_FN_Valet_ArmoredTruck_Parent
graph {
 properties {
  bUseMaterialAttributes=true
  bUsedWithMeshDeformer=true
  bUsedWithSkeletalMesh=true
  ditherOpacityMask=true
  opacityMaskClipValue=0.5
  pixelDepthOffsetMode=PDOM_Legacy
  shadingModel=MSM_FromMaterialExpression
  usedShadingModels=Default Lit | Clear Coat
 }
 nodes [
  n1 = PreSkinnedPosition()
  n2 = PreSkinnedLocalBounds()
  n3 = Divide(A=Subtract(A=VertexInterpolator(VS=@n1).PS, B=@n2.Min), B=Subtract(A=@n2.Max, B=@n2.Max))
  n4 = Round(None=ComponentMask(None=@n3) {r=true})
  n5 = MaterialFunctionCall() {materialFunction=/Valet/Materials/MF_Valet_BaseFunctionality}
  n6 = MaterialFunctionCall!(
   In=MaterialFunctionCall!(Mask=Multiply(A=@n4, B=@n5.Lights), MA=@n5.Result) {materialFunction=/Valet/Materials/MF_Valet_Headlights}.Result
   CenterLine=StaticSwitchParameter(True=ComponentMask(None=@n3) {g=true}, False=Constant()) {defaultValue=true, parameterName=UseCenterLine}
   Alpha=Multiply(A=@n5.Lights, B=OneMinus(None=@n4))
  ) {materialFunction=/Game/Weapons/WeaponSkins/MaterialFunction/TurnSignals}
  n7 = MaterialFunctionCall!(
   In=MaterialFunctionCall!(MA=MaterialFunctionCall!(Material=@n6.Result) {materialFunction=/Valet/Materials/MF_Valet_Decal}.Result) {materialFunction=/Valet/Materials/MF_Valet_ReverseLights}.Result
   Metal Tint=@n5.Scratch Color
   Local Bounds Split=@n3
  ) {materialFunction=/Game/Vehicles/Valet/MF_Valet_Damage}
  MaterialFunctionCall!(Material=@n7.Result, PreskinnedPos=@n1) {materialFunction=/Valet/Materials/MF_ValetMud}
  n8 = Add(A=NamedRerouteUsage(), B=Constant() {r=-0.3})
  Multiply(A=@n8, B=Constant() {r=1})
  n9 = Constant() {r=0.03}
  n10 = NamedRerouteUsage()
  n11 = Multiply(
   A=NamedRerouteUsage()
   B=MaterialFunctionCall(Color=GetMaterialAttributes(None=@n10) {attributeGetTypes [#1, #2]}.Base Color) {materialFunction=&SplitComponents}.G
  ) {constA=1}
  n12 = LocalPosition() {localOrigin=InstancePreSkinning}
  n13 = VertexInterpolator(VS=@n12.XYZ)
  n14 = Subtract(A=ComponentMask(None=ShaderStageSwitch(PixelShader=@n13.PS, VertexShader=@n12.XYZ)) {b=true}) {constB=320}
  NamedRerouteDeclaration(None=Multiply(A=Step(Y=@n9, X=@n11) {constY=0.2}, B=Saturate(None=@n14))) {name=TopLights_Mask, nodeColor=<rgba 0.9408 0.9304 1 1>}
  NamedRerouteDeclaration(None=@n5.Lights) {name=M_T_Lights, nodeColor=<rgba 0 1 0.3672 1>}
  NamedRerouteDeclaration(None=@n5.Emissive Mask) {name=M_T_Emissive, nodeColor=<rgba 1 0.0469 0 1>}
  NamedRerouteDeclaration(None=@n5.Gmap) {name=M_T_Gmap, nodeColor=<rgba 0 0.2734 1 1>}
  NamedRerouteDeclaration(None=ScalarParameter() {group=SirenAlarm, parameterName=RedLights, primitiveDataIndex=15}) {name=Prim_RedLights, nodeColor=<rgba 1 0 0.2344 1>}
  n15 = GetMaterialAttributes(None=NamedRerouteUsage()) {attributeGetTypes [#1, #2]}
  NamedRerouteDeclaration(None=@n15.Emissive Color) {name=M_In_Emissive, nodeColor=<rgba 0 1 0.1094 1>}
  n16 = Constant() {r=1}
  n17 = Constant() {r=-55}
  n18 = Distance(A=BumpOffset(Coordinate=TextureCoordinate(), Height=Constant() {r=0.2}), B=Constant2Vector() {g=0.043, r=0.2})
  NamedRerouteDeclaration(None=Saturate(None=LinearInterpolate(A=@n16, B=@n17, Alpha=@n18))) {name=Inner Bulb, nodeColor=<rgba 0.3359 0 1 1>}
  n19 = Subtract(A=PixelDepth().R, B=ScalarParameter() {defaultValue=1500, group=SirenAlarm, parameterName=DephtDistance})
  n20 = ScalarParameter() {defaultValue=500, group=SirenAlarm, parameterName=DephtFaloff}
  NamedRerouteDeclaration(None=Saturate(None=Divide(A=@n19, B=@n20))) {name=DephtMask, nodeColor=<rgba 1 0 0.5859 1>}
  NamedRerouteUsage()
  n21 = NamedRerouteUsage()
  n22 = VectorParameter() {group=SirenAlarm, parameterName=SirenAlarm_TintColor_1, defaultValue=<rgba 0.05 0.0007 0.0024 1>}
  n23 = VectorParameter() {group=SirenAlarm, parameterName=SirenAlarm_TintColor_2, defaultValue=<rgba 1.4 0.0672 0.056 1>}
  n24 = Multiply(A=NamedRerouteUsage(), B=Constant() {r=0.8})
  n25 = NamedRerouteUsage()
  n26 = Sine(None=Divide(A=Time(), B=ScalarParameter() {defaultValue=0.6, group=SirenAlarm, parameterName=Light_Sine_Speed}))
  n27 = ScalarParameter() {defaultValue=0.3, desc=.13, group=SirenAlarm, parameterName=Light_Sine Mult}
  n28 = Multiply(
   A=Add(A=Saturate(None=@n8), B=Multiply(A=@n24, B=@n25))
   B=Add(A=Multiply(A=@n26, B=@n27), B=ScalarParameter() {defaultValue=1, desc=.7, group=SirenAlarm, parameterName=Light_Sine bias})
  )
  n29 = LinearInterpolate(
   A=@n21
   B=LinearInterpolate(A=@n21, B=LinearInterpolate(A=@n22.RGB, B=@n23.RGB, Alpha=@n28), Alpha=NamedRerouteUsage())
   Alpha=NamedRerouteUsage()
  )
  n30 = Constant()
  n31 = NamedRerouteUsage()
  n32 = MaterialFunctionCall!(In=MaterialFunctionCall!(In=MaterialFunctionCall!(Material=@n7.Result) {materialFunction=/Valet/Materials/MF_ValetSpeakers}.Result) {materialFunction=/Game/Packages/Fortress_SharedMaterials/PartyBus_TempLocation/MF_ArmoredBattleBus_PartyLights}.Result) {materialFunction=/Valet/Materials/MF_OcclusionPrevention}
  n33 = MaterialAttributeLayers!(None=@n32.Result)
  n34 = MaterialFunctionCall!(In=MaterialFunctionCall(In=StaticSwitchParameter(True=@n33, False=@n32.Result) {parameterName=Use Material Layers}) {materialFunction=&MF_VehicleDeRez_Transition}.Result) {materialFunction=/VividRazorVaultGameplay/Assets/Gameplay_VaultDoor_A/Materials/MF_Tumblers}
  StaticSwitchParameter(
   True=SetMaterialAttributes!(
    MaterialAttributes=@n15.MaterialAttributes
    Emissive Color=EyeAdaptationInverse(LightValueInput=@n29).EyeAdaptationInverse
    Base Color=LinearInterpolate(A=@n15.Base Color, B=@n30, Alpha=@n31)
   ) {attributeSetTypes [#1, #2]}
   False=NamedRerouteDeclaration(None=MaterialFunctionCall!(In=@n34.Result) {materialFunction=&/MF_ArmoredTruck_Scortch}.Result) {name=OriginalMA, nodeColor=<rgba 0 0.0625 1 1>}
  ) {group=SirenAlarm, parameterName=UseSirenAlarm, sortPriority=-1}
 ]
}
