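# Physics asset for the armored truck's damaged left mirror mesh, going by its name, giving it a simple ragdoll-style setup.
# It defines two capsule bodies, a large one covering the truck body and a small one around the mirror_l bone.
# A single constraint joins mirror_l to body with limited swing and twist, keeping the mirror's angular movement within a narrow range.
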PhysicsAsset /ArmoredTruck/Assets/Meshes/ArmoredTruck_Veh_Dyn_dmg_mirror_l
bodies {
 body [{shapeType=Capsule, length=229.3724, radius=125.9405, center=<40.2655 -0.1455 66.4146>, rotation=<rot 0 -89.8532 -87.571>}]
 mirror_l [{
   shapeType=Capsule
   length=30.2632
   radius=23.2084
   shapeName=mirror_l_capsule
   center=<-7.2557 -31.6503 4.6957>
   rotation=<rot -72.1918 -50.78 -53.0341>
  }]
}
constraints [{bone1Name=mirror_l, bone2Name=body, swing1LimitDegrees=20, swing2LimitDegrees=30, twistLimitDegrees=20}]
