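# Physics asset for the armored truck's right mirror damage mesh, judging by its name, giving the mirror an articulated physics body.
# A large capsule stands in for the truck body and a small capsule covers the mirror_r bone.
# A single constraint joins mirror_r to body with narrow swing and twist angle limits, allowing the mirror only restricted rotation.
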PhysicsAsset /ArmoredTruck/Assets/Meshes/ArmoredTruck_Veh_Dyn_dmg_mirror_r
bodies {
 body [{shapeType=Capsule, length=229.3724, radius=125.9405, center=<40.2655 -0.1455 66.4146>, rotation=<rot 0 -89.8532 -87.571>}]
 mirror_r [{shapeType=Capsule, length=40, radius=20, shapeName=mirror_r_capsule, center=<0 29.083 -5.2506>, rotation=<rot 0 0 93.9647>}]
}
constraints [{bone1Name=mirror_r, bone2Name=body, swing1LimitDegrees=10, swing2LimitDegrees=30, twistLimitDegrees=20}]
