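# Physics asset for the roof lights of the dynamic armored truck vehicle, apparently for its damaged state.
# One large box body covers the truck body, and a smaller box sits on each of the four front and back, left and right roof light bones.
# Each light is constrained directly to the body with hinge-like limit values: wide on one swing axis, zero on the other and narrow twist, slightly looser at the front.
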PhysicsAsset /ArmoredTruck/Assets/Meshes/ArmoredTruck_Veh_Dyn_dmg_rooflights
bodies {
 body [{
   shapeType=Box
   extentX=269.5742
   extentY=290.4565
   extentZ=450
   shapeName=body_box
   center=<-81.2743 -2.4107 135.4224>
   rotation=<rot 89.7208 -0.05 -0.05>
  }]
 dyn_base_roof_light_bk_l [{
   shapeType=Box
   extentX=29.6367
   extentY=60.4759
   extentZ=39.247
   shapeName=dyn_base_roof_light_bk_l_box
   center=<14.6312 -15.5102 0>
   rotation=<rot -0.0003 -10.3654 0.0001>
  }]
 dyn_base_roof_light_bk_r [{
   shapeType=Box
   extentX=29.6229
   extentY=58.5132
   extentZ=37.3981
   shapeName=dyn_base_roof_light_bk_r_box
   center=<14.6312 -15.5102 0>
   rotation=<rot -0.0003 -10.3648 0>
  }]
 dyn_base_roof_light_fr_l [{
   shapeType=Box
   extentX=30.0668
   extentY=65.0389
   extentZ=43.9401
   shapeName=dyn_base_roof_light_fr_l_box
   center=<14.6312 -15.5102 0>
   rotation=<rot 0.0003 -10.3654 -0.0001>
  }]
 dyn_base_roof_light_fr_r [{
   shapeType=Box
   extentX=30.0667
   extentY=65.0382
   extentZ=43.939
   shapeName=dyn_base_roof_light_fr_r_box
   center=<14.6312 -15.5102 0>
   rotation=<rot 0.0003 -10.3649 0>
  }]
}
constraints [
 {bone1Name=dyn_base_roof_light_bk_l, bone2Name=body, swing1LimitDegrees=0, swing2LimitDegrees=100, twistLimitDegrees=10}
 {bone1Name=dyn_base_roof_light_bk_r, bone2Name=body, swing1LimitDegrees=0, swing2LimitDegrees=100, twistLimitDegrees=10}
 {bone1Name=dyn_base_roof_light_fr_l, bone2Name=body, swing1LimitDegrees=0, swing2LimitDegrees=100, twistLimitDegrees=13}
 {bone1Name=dyn_base_roof_light_fr_r, bone2Name=body, swing1LimitDegrees=0, swing2LimitDegrees=100, twistLimitDegrees=13}
]
