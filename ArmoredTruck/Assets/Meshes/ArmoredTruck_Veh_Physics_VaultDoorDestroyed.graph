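# Physics asset for the armored truck vehicle in its vault-door-destroyed state.
# A single frame body with no constraints is built from boxes for the floor, front, cabin rear wall, cargo sides, upper cabin and tilted window corners.
# No box closes the back or roof of the cargo area, leaving it open.
# Upper and side shapes carry a Query_ name prefix, while the floor is named as physics.
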PhysicsAsset /ArmoredTruck/Assets/Meshes/ArmoredTruck_Veh_Physics_VaultDoorDestroyed
bodies {frame [
  {shapeType=Box, extentX=523.3182, extentY=286.5332, extentZ=10.2256, shapeName=bottom_physics, center=<-47.0648 0 40.1178>}
  {shapeType=Box, extentX=7.5, extentY=305.9886, extentZ=254.0595, shapeName=cabin_rear, center=<62.0449 0 162.0597>}
  {shapeType=Box, extentX=230.3774, extentY=257.3701, extentZ=109.9021, shapeName=front_bottom, center=<228.2586 0 86.9829>}
  {shapeType=Box, extentX=419.5, extentY=11.1723, extentZ=264.4772, shapeName=Query_passenger_side, center=<-90.1214 132.0863 154.0272>}
  {shapeType=Box, extentX=419.5, extentY=11.1723, extentZ=264.4772, shapeName=Query_driver_side, center=<-89.9831 -125.0152 154.0272>}
  {shapeType=Box, extentX=152.8193, extentY=235.5311, extentZ=44.2552, shapeName=Query_front_mid, center=<195.7095 1.7142 163.9477>}
  {shapeType=Box, extentX=71.071, extentY=235.5311, extentZ=17.9499, shapeName=Query_front_top, center=<148.1432 1.7142 247.6173>}
  {
   shapeType=Box
   extentX=7.8756
   extentY=10.3367
   extentZ=64.5643
   shapeName=Query_driver_window_corner
   center=<192.1315 -104.3883 221.8031>
   rotation=<rot 24.9486 -0.0003 -0.0006>
  }
  {
   shapeType=Box
   extentX=10.0805
   extentY=10.3367
   extentZ=64.5643
   shapeName=Query_passenger_window_corner
   center=<192.8641 106.6559 217.4938>
   rotation=<rot 24.9486 -0.0003 -0.0006>
  }
 ]}
constraints []
